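# Parent material for the mask on the RowLiaison Cast female character body's Superlevel skins, built on the shared 2023 character parent functions.
# MF_CharacterParent_2023_PreFX builds the material attributes, then Base Color is multiplied by a DiffuseColorTint parameter in the Superlevel group, white by default.
# The tinted attributes feed MF_CharacterParent_2023_PostFX along with PreFX's auxiliary outputs.
# Shading model comes from the material expression.
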Material /BRCosmetics/Characters/Player/Female/Medium/Bodies/F_MED_RowLiaison_Cast/Skins/Superlevels/Materials/T1/M_RowLiaison_Cast_Mask_PARENT
graph {
 properties {
  bAutomaticallySetUsageInEditor=false
  bUseMaterialAttributes=true
  bUsedWithClothing=true
  bUsedWithMeshDeformer=true
  bUsedWithMorphTargets=true
  bUsedWithNanite=true
  bUsedWithNiagaraMeshParticles=true
  bUsedWithSkeletalMesh=true
  pixelDepthOffsetMode=PDOM_Legacy
  shadingModel=MSM_FromMaterialExpression
  usedShadingModels=Default Lit | Subsurface | Clear Coat | Subsurface Profile | Cloth
 }
 nodes [
  n1 = MaterialFunctionCall() {materialFunction=&MF_CharacterParent_2023_PreFX}
  n2 = GetMaterialAttributes(None=@n1.Result) {attributeGetTypes [#1]}
  n3 = VectorParameter() {group=Superlevel, parameterName=DiffuseColorTint, defaultValue=<rgba 1 1 1 0>}
  MaterialFunctionCall(
   Material=SetMaterialAttributes!(MaterialAttributes=@n2.MaterialAttributes, Base Color=Multiply(A=@n2.Base Color, B=@n3.RGB)) {attributeSetTypes [#1]}
   M=@n1.M
   Pre Skinned Local Position=@n1.M
   Distance Blend=@n1.M
   Fresnel=@n1.M
  ) {materialFunction=&MF_CharacterParent_2023_PostFX}
 ]
}
